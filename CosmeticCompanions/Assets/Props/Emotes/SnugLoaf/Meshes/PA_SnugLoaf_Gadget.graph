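# Physics asset for the SnugLoaf gadget mesh, a cosmetic companion emote prop.
# It has a single body on the root bone, built from one wide, short, tilted capsule offset upward from the bone origin.
# With no constraints, the prop is represented as one rigid collision volume with no articulated parts.
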PhysicsAsset /CosmeticCompanions/Assets/Props/Emotes/SnugLoaf/Meshes/PA_SnugLoaf_Gadget
bodies {root [{
   shapeType=Capsule
   length=20.2706
   radius=52.8156
   shapeName=root_capsule
   center=<1.9984 0.8235 22.3251>
   rotation=<rot 0 -42.7111 -88.2703>
  }]}
constraints []
